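# Sound cue for consuming the CubeFlopper consumable.
# A random node picks one of three consume wave variations with equal weight.
# It plays at reduced volume in the consumables sound class, spatialized through the shared Athena consumable-use attenuation asset.
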SoundCue /CorruptionGameplay/Sounds/Consumables/CubeFlopper/CubeFlopper_Consume_Cue
root=Random_0
nodes [
 {id=Random_0, children [WavePlayer_0,WavePlayer_1,WavePlayer_2], properties {weights [1,1,1]}}
 {id=WavePlayer_0, properties {soundWaveAssetPtr=&/CubeFlopper_Consume_01}}
 {id=WavePlayer_1, properties {soundWaveAssetPtr=&/CubeFlopper_Consume_02}}
 {id=WavePlayer_2, properties {soundWaveAssetPtr=&/CubeFlopper_Consume_03}}
]
properties {
 attenuationSettings=&Fort_Consumable_Use_Athena_Att
 maxDistance=2250
 soundClassObject=&Fort_Consumables
 volumeMultiplier=0.6
}
