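# CubeFlopper_Equip_Cue is a SoundCue for the equip sound of the CubeFlopper consumable in CorruptionGameplay.
# A single wave player feeds the CubeFlopper_Equip sound wave straight to the output, with no randomization or mixing.
# It uses the Athena consumable-use attenuation and the loot collect sound class, with reduced volume and a set maximum distance.
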SoundCue /CorruptionGameplay/Sounds/Consumables/CubeFlopper/CubeFlopper_Equip_Cue
root=WavePlayer_0
nodes [{id=WavePlayer_0, properties {soundWaveAssetPtr=&/CubeFlopper_Equip}}]
properties {
 attenuationSettings=&Fort_Consumable_Use_Athena_Att
 maxDistance=2250
 soundClassObject=&Fort_Loot_Collect
 volumeMultiplier=0.3
}
