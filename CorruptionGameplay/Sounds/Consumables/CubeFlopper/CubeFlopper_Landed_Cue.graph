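# CubeFlopper_Landed_Cue is a sound cue for the CubeFlopper consumable landing, going by its name and folder.
# A single wave player feeds the CubeFlopper_Landed wave straight to the root, with no randomization or modulation.
# The cue plays at reduced volume, routed through the loot collect sound class with weapon drop attenuation for spatial falloff.
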SoundCue /CorruptionGameplay/Sounds/Consumables/CubeFlopper/CubeFlopper_Landed_Cue
root=WavePlayer_0
nodes [{id=WavePlayer_0, properties {soundWaveAssetPtr=&/CubeFlopper_Landed}}]
properties {
 attenuationSettings=&Fort_Weapon_Drop_Att
 maxDistance=3000
 soundClassObject=&Fort_Loot_Collect
 volumeMultiplier=0.5
}
